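# M_IngameMap_Creative is an unlit translucent UI material, named for the Creative variant of the in-game map.
# It contains a texture sample of T_IngameMap_Target set to explicit mip level mode, plus a constant of one.
# No connections between these nodes or to the material outputs are shown.
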Material /Game/UI/IngameMap/M_IngameMap_Creative
graph {
 properties {
  bUsedWithStaticMesh=false
  blendMode=BLEND_Translucent
  materialDomain=MD_UI
  pixelDepthOffsetMode=PDOM_Legacy
  shadingModel=MSM_Unlit
  usedShadingModels=Unlit
 }
 nodes [
  Constant() {r=1}
  TextureSample!(UVs, Tex, Level, Apply View MipBias) {mipValueMode=TMVM_MipLevel, texture=&/T_IngameMap_Target}
 ]
}
