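# Material function that renders the Playground (Creative) thermometer heatmap UI over an input color.
# The red channel of a heatmap render target drives a four-color ramp, mixed evenly with the partly desaturated input.
# A rotated grid, sized by a material parameter collection, is screen-blended on top.
# Scrolling, pulsing icons are then added: exclamation marks where red is near maximum, lock symbols where the green channel is set.
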MaterialFunction /Game/Athena/HUD/LTM/Playground/Thermometer/MF_UI_creativeheatmap
graph {nodes [
  n1 = Multiply(
   A=MaterialFunctionCall(UVs=TextureCoordinate(), "Rotation Angle (0-1)"=ScalarParameter() {defaultValue=0.25, parameterName=Grid Rotation Angle}) {materialFunction=&CustomRotator}.Rotated Values
   B=CollectionParameter() {collection=&/MPC_heatmap, parameterName=Gridsize}
  )
  n2 = Desaturation(None=FunctionInput(Preview=Constant3Vector()), Fraction=Constant() {r=0.6})
  n3 = Constant() {r=0.5}
  n4 = TextureSample() {
   samplerSource=SSM_Clamp_WorldGroupSettings
   samplerType=SAMPLERTYPE_LinearColor
   texture=&/RT_Heatmapcode
  }
  n5 = Multiply(
   A=MaterialFunctionCall(
    1=VectorParameter() {parameterName=Heatmap Color 0, defaultValue=<rgba 0 0 2.5 0>}.RGB
    2=VectorParameter() {parameterName=Heatmap Color 1, defaultValue=<rgba 0 1 0 0>}.RGB
    3=VectorParameter() {parameterName=Heatmap Color 2, defaultValue=<rgba 1 0.678 0 0>}.RGB
    4=VectorParameter() {parameterName=Heatmap Color 3, defaultValue=<rgba 1 0.03 0.03 0>}.RGB
    A=@n4.R
   ) {materialFunction=&Lerp_Multiple_Float3}.Lerp 4 Inputs
   B=ScalarParameter() {defaultValue=1.1, parameterName=Heatmap Brightness}
  ) {constB=0.8}
  n6 = MaterialFunctionCall!(
   Blend=Multiply(
    A=TextureSample(UVs=@n1) {texture=&/T_Lock_group}.G
    B=ScalarParameter() {defaultValue=0.25, parameterName=Grid Brightness}
   ) {constB=0.25}
   Base=Saturate(None=LinearInterpolate(A=@n5, B=@n2, Alpha=@n3))
  ) {materialFunction=/Engine/Functions/Engine_MaterialFunctions03/Blends/Blend_Screen}
  n7 = TextureSample() {
   samplerSource=SSM_Clamp_WorldGroupSettings
   samplerType=SAMPLERTYPE_LinearColor
   texture=&RT_Heatmapcode
  }
  n8 = TextureSample() {
   samplerSource=SSM_Clamp_WorldGroupSettings
   samplerType=SAMPLERTYPE_LinearColor
   texture=&RT_Heatmapcode
  }
  n9 = Multiply(A=@n8.R) {constB=10}
  n10 = Panner(Coordinate=TextureCoordinate() {uTiling=32, vTiling=32}, Speed=Constant() {r=-0.1}) {speedX=1}
  n11 = Sine(None=Time() {bOverride_Period=true, period=3})
  n12 = Multiply(
   A=Multiply(A=Saturate(None=@n9), B=TextureSample(UVs=@n10) {texture=&/T_Exclamaition}.R)
   B=LinearInterpolate(Alpha=ConstantBiasScale(None=@n11)) {constA=0.25}
  )
  n13 = TextureSample() {
   samplerSource=SSM_Clamp_WorldGroupSettings
   samplerType=SAMPLERTYPE_LinearColor
   texture=&RT_Heatmapcode
  }
  n14 = TextureSample(UVs=Panner(Coordinate=TextureCoordinate() {uTiling=32, vTiling=32}, Speed=Constant() {r=-0.1}) {speedX=1}) {texture=&/T_LockscreenLocked}
  n15 = ConstantBiasScale(None=Sine(None=Time() {bOverride_Period=true, period=3}))
  n16 = TextureSample() {
   samplerSource=SSM_Clamp_WorldGroupSettings
   samplerType=SAMPLERTYPE_LinearColor
   texture=&RT_Heatmapcode
  }
  n17 = LinearInterpolate(
   A=If(A=@n7.R, B=Constant() {r=0.9}, "A > B"=@n12, "A == B"=@n12, "A < B"=Constant())
   B=Multiply(A=Multiply(A=@n13.G, B=@n14.R), B=LinearInterpolate(Alpha=@n15) {constA=0.25})
   Alpha=@n16.G
  )
  FunctionOutput(None=Saturate(None=Add(A=@n6.Result, B=@n17)))
 ]}
