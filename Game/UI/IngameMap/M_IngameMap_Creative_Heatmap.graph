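# UI material for the in-game map's Creative heatmap, translucent and unlit.
# It samples the in-game map target texture in explicit mip-level mode and passes the RGB into the shared MF_UI_creativeheatmap function.
# The heatmap logic lives in that function, which sits under the LTM Playground Thermometer HUD folder, so this material is a thin wrapper around it.
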Material /Game/UI/IngameMap/M_IngameMap_Creative_Heatmap
graph {
 properties {
  bUsedWithStaticMesh=false
  blendMode=BLEND_Translucent
  materialDomain=MD_UI
  pixelDepthOffsetMode=PDOM_Legacy
  shadingModel=MSM_Unlit
  usedShadingModels=Unlit
 }
 nodes [
  Constant() {r=1}
  MaterialFunctionCall!(In=TextureSample!(UVs, Tex, Level, Apply View MipBias) {mipValueMode=TMVM_MipLevel, texture=&/T_IngameMap_Target}.RGB) {materialFunction=/Game/Athena/HUD/LTM/Playground/Thermometer/MF_UI_creativeheatmap}
 ]
}
