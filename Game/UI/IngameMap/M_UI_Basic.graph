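# Basic translucent UI-domain material, stored with the in-game map UI assets.
# It exposes a Color vector parameter, white by default, and an Opacity scalar parameter, fully opaque by default, for material instances to override.
# The graph contains no textures or math, and automatic usage-flag setting in the editor is turned off.
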Material /Game/UI/IngameMap/M_UI_Basic
graph {
 properties {
  bAutomaticallySetUsageInEditor=false
  bUsedWithStaticMesh=false
  blendMode=BLEND_Translucent
  materialDomain=MD_UI
  usedShadingModels=Default Lit
 }
 nodes [
  ScalarParameter() {defaultValue=1, parameterName=Opacity}
  VectorParameter() {parameterName=Color, defaultValue=<rgba 1 1 1 1>}
 ]
}
