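# Translucent UI master material for an animated glitch effect, named for text on the in-game map.
# Two panning noise samples on aspect-corrected UVs, one distorting the other, drive time-offset sine waves that combine into a shifting block mask.
# That mask shapes a blend between two colors and a fill opacity pulsed by time-sampled blurred noise.
# UseSineMask optionally mixes in irregular flicker from three combined sine timers.
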Material /Game/UI/IngameMap/M_UI_MapGlitchText_Master
graph {
 properties {
  bUsedWithStaticMesh=false
  blendMode=BLEND_Translucent
  materialDomain=MD_UI
  pixelDepthOffsetMode=PDOM_Legacy
  usedShadingModels=Default Lit
 }
 nodes [
  n1 = ScalarParameter() {defaultValue=0.5, parameterName=FillAmount}
  n2 = Multiply(A=@n1, B=ScalarParameter() {defaultValue=1, parameterName=FillColorCorrection})
  n3 = ScalarParameter() {defaultValue=0.5, parameterName=Time_4}
  n4 = Power(
   Base=TextureSample(UVs=MaterialFunctionCall(Speed=@n3) {materialFunction=&DebugTimeSine}.Result) {samplerType=SAMPLERTYPE_Grayscale, texture=&LowResBlurredNoise}.R
   Exp=ScalarParameter() {defaultValue=0.5, parameterName=PulsePower}
  )
  DotProduct(A=@n2, B=@n4)
  n5 = Time()
  n6 = ScalarParameter() {defaultValue=5, group=Texture1, parameterName=BlocksTimeScale}
  n7 = MaterialFunctionCall(
   UVs=MaterialFunctionCall() {materialFunction=/FortUILibrary/Foundation/Art_ToolKit/Tools/MF_UI_SquareToScreenSpace}.X_Priority
   Texture Scale=ScalarParameter() {defaultValue=1, parameterName=Scale}
  ) {materialFunction=&ScaleUVsByCenter}
  n8 = Multiply(A=@n7.UVs, B=AppendVector(A=Constant() {r=1}, B=ScalarParameter() {defaultValue=1, group=Texture1, parameterName=Y_Scale_Correction}))
  n9 = ScalarParameter() {defaultValue=0.5, group=Texture1, parameterName=Tile}
  n10 = ScalarParameter() {defaultValue=0.06, group=Texture1, parameterName=PanSpeed_Noise_X}
  n11 = ScalarParameter() {defaultValue=0.05, group=Texture1, parameterName=PanSpeed_Noise_Y}
  n12 = Multiply(A=@n7.UVs, B=AppendVector(A=Constant() {r=1}, B=ScalarParameter() {defaultValue=1, group=Texture2, parameterName=Y_Scale_Correction_2}))
  n13 = ScalarParameter() {defaultValue=3, group=Texture2, parameterName=Tile_Noise_2}
  n14 = ScalarParameter() {defaultValue=-0.125, group=Texture2, parameterName=PanSpeed_Noise_X_2}
  n15 = ScalarParameter() {defaultValue=-0.125, group=Texture2, parameterName=PanSpeed_Noise_Y_2}
  n16 = TextureSampleParameter2D(UVs=Panner(Coordinate=Multiply(A=@n12, B=@n13), Speed=AppendVector(A=@n14, B=@n15)) {speedX=-0.1}) {parameterName=Texture, samplerType=SAMPLERTYPE_LinearColor, texture=&/T_UI_MapGlitchText_Noise_512x}
  n17 = Add(
   A=Panner(Coordinate=Multiply(A=@n8, B=@n9), Speed=AppendVector(A=@n10, B=@n11)) {speedX=-0.1}
   B=Multiply(A=AppendVector(A=@n16.R, B=@n16.R), B=ScalarParameter() {defaultValue=0.25, group=Texture1, parameterName=AdditionalDistortion})
  )
  n18 = Sine(None=Add(
    A=Divide(A=@n5, B=@n6)
    B=TextureSampleParameter2D(UVs=@n17) {parameterName=Texture, samplerType=SAMPLERTYPE_LinearColor, texture=&T_UI_MapGlitchText_Noise_512x}.R
   )) {period=0.15}
  n19 = Divide(A=Time(), B=ScalarParameter() {defaultValue=16, group=Texture2, parameterName=BlocksTimeScale_2})
  n20 = ConstantBiasScale(None=Sine(None=Add(A=@n19, B=@n16.B)) {period=0.15})
  n21 = Saturate(None=DotProduct(A=ConstantBiasScale(None=@n18), B=OneMinus(None=@n20)))
  LinearInterpolate(
   A=VectorParameter() {group=Color, parameterName=Color2, defaultValue=<rgba 0 0.5 1 1>}.RGB
   B=VectorParameter() {group=Color, parameterName=Color1, defaultValue=<rgba 0.25 1 0.5 1>}.RGB
   Alpha=Saturate(None=LinearInterpolate(A=@n2, Alpha=@n21))
  )
  n22 = Saturate(None=LinearInterpolate(A=DotProduct(A=@n1, B=@n4), Alpha=@n21))
  n23 = ScalarParameter() {defaultValue=4, parameterName=TimeExtension}
  n24 = LinearInterpolate(
   B=@n23
   Alpha=MaterialFunctionCall(Speed=ScalarParameter() {defaultValue=0.5, parameterName=Time_1}) {materialFunction=&DebugTimeSine}.Result
  ) {constB=4}
  n25 = LinearInterpolate(
   B=@n23
   Alpha=MaterialFunctionCall(Speed=ScalarParameter() {defaultValue=0.5, parameterName=Time_2}) {materialFunction=&DebugTimeSine}.Result
  ) {constB=4}
  n26 = ScalarParameter() {defaultValue=0.5, parameterName=Time_3}
  n27 = DotProduct(
   A=Saturate(None=DotProduct(A=@n24, B=@n25))
   B=LinearInterpolate(B=@n23, Alpha=MaterialFunctionCall(Speed=@n26) {materialFunction=&DebugTimeSine}.Result) {constB=4}
  )
  LinearInterpolate(A=@n22, B=LinearInterpolate(A=@n22, Alpha=Saturate(None=@n27)), Alpha=ScalarParameter() {parameterName=UseSineMask})
  Multiply()
 ]
}
